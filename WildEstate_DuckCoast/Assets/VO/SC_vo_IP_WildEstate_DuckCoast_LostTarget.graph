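# Sound cue for the LostTarget voice line of the WildEstate DuckCoast NPC, likely played when it loses track of its target.
# A Random node picks one of seven equally weighted voice recordings.
# Playback uses shared henchman NPC attenuation, the critical enemy VO sound class, and boss combat chatter concurrency to control overlapping lines.
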SoundCue /WildEstate_DuckCoast/Assets/VO/SC_vo_IP_WildEstate_DuckCoast_LostTarget
root=Random_1
nodes [
 {
  id=Random_1
  children [WavePlayer_0,WavePlayer_1,WavePlayer_2,WavePlayer_3,WavePlayer_4,WavePlayer_5,WavePlayer_6]
  properties {weights [1,1,1,1,1,1,1]}
 }
 {id=WavePlayer_0, properties {soundWaveAssetPtr=/WildEstate_DuckCoast/Assets/VO/wav/vo_IP_WildEstate_DuckCoast_LostTarget_010}}
 {id=WavePlayer_1, properties {soundWaveAssetPtr=/WildEstate_DuckCoast/Assets/VO/wav/vo_IP_WildEstate_DuckCoast_LostTarget_020}}
 {id=WavePlayer_2, properties {soundWaveAssetPtr=/WildEstate_DuckCoast/Assets/VO/wav/vo_IP_WildEstate_DuckCoast_LostTarget_030}}
 {id=WavePlayer_3, properties {soundWaveAssetPtr=/WildEstate_DuckCoast/Assets/VO/wav/vo_IP_WildEstate_DuckCoast_LostTarget_040}}
 {id=WavePlayer_4, properties {soundWaveAssetPtr=/WildEstate_DuckCoast/Assets/VO/wav/vo_IP_WildEstate_DuckCoast_LostTarget_050}}
 {id=WavePlayer_5, properties {soundWaveAssetPtr=/WildEstate_DuckCoast/Assets/VO/wav/vo_IP_WildEstate_DuckCoast_LostTarget_060}}
 {id=WavePlayer_6, properties {soundWaveAssetPtr=/WildEstate_DuckCoast/Assets/VO/wav/vo_IP_WildEstate_DuckCoast_LostTarget_070}}
]
properties {
 attenuationSettings=/Game/Sounds/Fort_Human_VO_BR/NPCs/Shared/NPC_VO_Henchmen_Att
 maxDistance=8500
 soundClassObject=&Fort_VO_Enemy_Critical
 volumeMultiplier=1
 concurrencySet [&NPC_Boss_Combat_Chatter_Conc]
}
